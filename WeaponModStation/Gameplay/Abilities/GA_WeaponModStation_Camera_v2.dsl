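; Gameplay ability that switches the player's camera to a dedicated mode, apparently for the weapon mod station.
; On activation it pushes the configured camera mode class if the avatar actor is valid, otherwise it ends the ability immediately.
; When the ability ends, it pops the camera mode through a DoOnce, returning to normal gameplay view.
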(event Ability|EventActivateAbility
  ;[0.4134 0 1] GA is used to push a new camera
  (Utilities|IsValid (Ability|GetAvatarActorfromActorInfo)
    (:Is%20Valid
      (Camera|PushCameraMode (Variables|Default|GetCamera%20Mode%20Class)))
    (:Is%20Not%20Valid
      (Ability|EndAbility))))

(event Ability|EventOnEndAbility
  (params (bWasCancelled :type Boolean))
  ;[0 0.03758 1] End the ability, pop the camera back to normal gameplay
  (Utilities|FlowControl|DoOnce false
    (:Completed
      (Camera|PopCameraMode))))
